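# Niagara rain effect for the Precipitation emote, with GPU and CPU versions of both raindrops and ground ripples.
# Drops fall from a bobbing overhead box, stretch with speed and age out on collision, with user-set area, rate and color.
# Ripples spawn independently at ground level, storing a static position alongside gravity-driven motion for two sprite renderers.
# Camera-distance remaps and front-end modifiers tune visibility; ripple tint follows time of day.
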NiagaraSystem /BRCosmetics/Animation/Game/MainPlayer/Emotes/Precipitation/CMM/FX/NS_Emote_Precipitation_Rain
dependencies {
 usedDataInterfaces [
  &NiagaraDataInterfaceCurve
  &NiagaraDataInterfaceSpriteRendererInfo
  &NiagaraDataInterfacePlatformSet
 ]
 usedDynamicInputs [&UniformRangedFloat]
 usedModules [&SystemState]
}
emitters {
 Raindrops_GPU {
  simTarget=GPUComputeSim
  emitterUpdateScript [
   &EmitterState()
   &SpawnRate(
    SpawnRate=&Multiply_Float(
     A=&FN_LOD_Distance_Remap(Start Distance=250, End Distance=2500, Minimum Output=200, Maximum Output=50, Clamp Output=true)
     B=$User.SpawnRate
    )
    Use Spawn Probability=false
   )
  ]
  particleSpawnScript [
   &InitializeParticle(
    Lifetime Max=0.8
    Lifetime Min=0.5
    Lifetime Mode=::NewEnumerator1
    Sprite Size Max=<3.75 4.5>
    Sprite Size Min=<3 3.3>
    Sprite Size Mode=::NewEnumerator2
    UsePositionOffset=false
    Write Material Random=true
   ) {module=InitializeParticle001}
   &AddVelocity(Velocity=&Multiply_Vector(A=&RandomRangeVector(Minimum=<-300 -300 -1200>, Maximum=<300 300 -2240>), B=<0 0 0.5>))
   /SetVariables(Particles.Color=&LerpLinearColors(
     StartColor=<rgba 0.638 0.7461 0.8125 1>
     EndColor=&MultiplyLinearColor(Linear Color=$User.Color, Scale Factor=&MakeVector4_FromFloat(Float=$User.CustomColorMult))
     LerpFactor=$User.UseCustomColor
    )) {isSetParameters=true, module=SetVariables_#1}
   &Color(Color=&MakeLinearColorFromVectorAndFloat("Vector (RGB)"=<0.6281 0.5529 0.974>, "Float (Alpha)"=&ClampFloat(Float=1.2))) {enabled=false}
   &SphereLocation(
    MaskSpawnGroup=false
    Override Seed=false
    Sphere Radius=0
    Surface Only=false
    UseNonUniformScale=false
    UseOffset=false
    UseSphereOrientation=false
   ) {module=SphereLocation001}
   &BoxLocation(
    Box Offset=&MakeVector(
     X=0
     Y=0
     Z=&Divide_Float(
      A=&Lerp_Float(A=300, B=350, Alpha=HlslExpression() {hlslExpression="sin(Emitter.Age)"})
      B=&MakeFloatFromVector(Vector=$Engine.Owner.Scale, Channel=::NewEnumerator2)
     )
    )
    Box Size=$User.BoxSize
    MaskSpawnGroup=false
    Module.Override Seed=false
    Rotate Box=false
    Surface Only=false
   )
  ]
  particleUpdateScript [
   &ParticleState()
   &Collision(
    Advanced Aging Rate=100000
    Age Colliding Particles=true
    Enable Rest State=false
    GPU Collision Type=::NewEnumerator2
    Method for Calculating Particle Radius:ENiagaraCalculateRadiusOptions=::NewEnumerator2
    Module.EnableMaxCollisionCount=true
    Module.Randomize Collision Normal Vector=true
    Restitution=0
   )
   &GravityForce()
   &SolveForcesAndVelocity(Clamp Velocity=false, Limit Acceleration=false)
   &ScaleSpriteSizeBySpeed(Max Scale Factor=<1 15>, Source Velocity:Vector3f=&Multiply_Vector(A=<0.05 0.05 1>, B=$Particles.Velocity))
   &ScaleSpriteSize(
    Uniform Curve Scale=&Multiply_Float(
     A=&Multiply_Float(
      A=0.3
      B=&FN_LOD_Distance_Remap(Start Distance=250, End Distance=2500, Minimum Output=1, Maximum Output=4, Clamp Output=true)
     )
     B=&FN_FrontEndFloatModifier(Gameplay=1, Front End=1.3)
    )
    Uniform Curve Sprite Scale=DataInterface() {propertyValues {Curve {keys [{tangentMode=RCTM_User, time=0.5, value=1}, {tangentMode=RCTM_User, time=1, value=0}]}}}
   )
   &ScaleColor(
    Scale Alpha=&FloatFromCurve(
     FloatCurve=DataInterface() {propertyValues {Curve {keys [
         {interpMode=RCIM_Cubic, tangentMode=RCTM_User, time=0, value=0}
         {interpMode=RCIM_Cubic, tangentMode=RCTM_User, time=0.05, value=1}
         {arriveTangent=0.106, interpMode=RCIM_Cubic, leaveTangent=0.106, tangentMode=RCTM_User, time=0.6798, value=1}
         <at 1 0 RCIM_Cubic>
        ]}}}
     Scale Curve=&Multiply_Float(
      A=&Multiply_Float(
       A=&Lerp_Float(A=1, B=0.5, Alpha=$User.UseCustomColor)
       B=&FN_LOD_Distance_Remap(Start Distance=250, End Distance=2500, Minimum Output=1, Maximum Output=2, Clamp Output=true)
      )
      B=&FN_FrontEndFloatModifier(Gameplay=1, Front End=1.35)
     )
    )
    Scale RGB=&VectorFromFloat(Value=HlslExpression() {hlslExpression=2 + User.UseCustomColor * 2})
    ScaleA=true
    ScaleRGB=true
   )
  ]
  rendererClasses [&NiagaraSpriteRendererProperties]
  renderers [{rendererClass=&NiagaraSpriteRendererProperties}]
 }
 Raindrops_Low {
  emitterUpdateScript [
   &EmitterState()
   &SpawnRate(SpawnRate=&Multiply_Float(A=20, B=$User.SpawnRate), Use Spawn Probability=false)
  ]
  particleSpawnScript [
   &InitializeParticle(
    Lifetime Max=0.8
    Lifetime Min=0.5
    Lifetime Mode=::NewEnumerator1
    Sprite Size Max=<3.75 4.5>
    Sprite Size Min=<3 3.3>
    Sprite Size Mode=::NewEnumerator2
    UsePositionOffset=false
    Write Material Random=true
   ) {module=InitializeParticle001}
   &AddVelocity(Velocity=&Multiply_Vector(A=&RandomRangeVector(Minimum=<-300 -300 -1200>, Maximum=<300 300 -2240>), B=<0 0 0.5>))
   /SetVariables(Particles.Color=&LerpLinearColors(
     StartColor=<rgba 0.638 0.7461 0.8125 1>
     EndColor=&MultiplyLinearColor(Linear Color=$User.Color, Scale Factor=&MakeVector4_FromFloat(Float=$User.CustomColorMult))
     LerpFactor=$User.UseCustomColor
    )) {isSetParameters=true, module=SetVariables_#1}
   &Color(Color=&MakeLinearColorFromVectorAndFloat("Vector (RGB)"=<0.6281 0.5529 0.974>, "Float (Alpha)"=&ClampFloat(Float=1.2))) {enabled=false}
   &SphereLocation(
    MaskSpawnGroup=false
    Override Seed=false
    Sphere Radius=0
    Surface Only=false
    UseNonUniformScale=false
    UseOffset=false
    UseSphereOrientation=false
   ) {module=SphereLocation001}
   &BoxLocation(
    Box Offset=&MakeVector(
     X=0
     Y=0
     Z=&Divide_Float(
      A=&Lerp_Float(A=300, B=350, Alpha=HlslExpression() {hlslExpression="sin(Emitter.Age)"})
      B=&MakeFloatFromVector(Vector=$Engine.Owner.Scale, Channel=::NewEnumerator2)
     )
    )
    Box Size=$User.BoxSize
    MaskSpawnGroup=false
    Module.Override Seed=false
    Rotate Box=false
    Surface Only=false
   )
  ]
  particleUpdateScript [
   &ParticleState()
   &Collision(
    Advanced Aging Rate=100000
    Age Colliding Particles=true
    Enable Rest State=false
    Method for Calculating Particle Radius:ENiagaraCalculateRadiusOptions=::NewEnumerator2
    Module.EnableMaxCollisionCount=true
    Module.Randomize Collision Normal Vector=true
    Restitution=0.3
   )
   &GravityForce()
   &SolveForcesAndVelocity(Clamp Velocity=false, Limit Acceleration=false)
   &ScaleSpriteSizeBySpeed(Max Scale Factor=<1 15>, Source Velocity:Vector3f=&Multiply_Vector(A=<0.05 0.05 1>, B=$Particles.Velocity))
   &ScaleSpriteSize(
    Uniform Curve Scale=&Multiply_Float(
     A=&Multiply_Float(
      A=0.3
      B=&FN_LOD_Distance_Remap(Start Distance=250, End Distance=2500, Minimum Output=1, Maximum Output=4, Clamp Output=true)
     )
     B=&FN_FrontEndFloatModifier(Gameplay=1, Front End=1.2)
    )
    Uniform Curve Sprite Scale=DataInterface() {propertyValues {Curve {keys [{tangentMode=RCTM_User, time=0.5, value=1}, {tangentMode=RCTM_User, time=1, value=0}]}}}
   )
   &ScaleColor(
    Scale Alpha=&FloatFromCurve(
     FloatCurve=DataInterface() {propertyValues {Curve {keys [
         {interpMode=RCIM_Cubic, tangentMode=RCTM_User, time=0, value=0}
         {interpMode=RCIM_Cubic, tangentMode=RCTM_User, time=0.05, value=1}
         {arriveTangent=0.106, interpMode=RCIM_Cubic, leaveTangent=0.106, tangentMode=RCTM_User, time=0.6798, value=1}
         <at 1 0 RCIM_Cubic>
        ]}}}
     Scale Curve=&Multiply_Float(
      A=&Multiply_Float(
       A=&Lerp_Float(A=1, B=0.5, Alpha=$User.UseCustomColor)
       B=&FN_LOD_Distance_Remap(Start Distance=250, End Distance=2500, Minimum Output=1, Maximum Output=2, Clamp Output=true)
      )
      B=&FN_FrontEndFloatModifier(Gameplay=1, Front End=1.35)
     )
    )
    Scale RGB=&VectorFromFloat(Value=HlslExpression() {hlslExpression=2 + User.UseCustomColor * 2})
    ScaleA=true
    ScaleRGB=true
   )
   &ScaleColor(
    Scale Alpha=&MakeCustomFloatFromBool(
     Bool=&IsPlatformSetActive(New Platform Set=DataInterface() {propertyValues {Platforms {qualityLevelMask=-30}}})
     True Float=&Lerp_Float(
      A=1
      B=10
      Alpha=&FN_LOD_Distance_Remap(Start Distance=100, End Distance=3000, Minimum Output=0, Maximum Output=1, Clamp Output=true)
     )
     False Float=1
    )
    Scale RGB=&VectorFromFloat(Value=&MakeCustomFloatFromBool(
      Bool=&IsPlatformSetActive(New Platform Set=DataInterface() {propertyValues {Platforms {qualityLevelMask=-30}}})
      True Float=&Lerp_Float(
       A=1
       B=2
       Alpha=&FN_LOD_Distance_Remap(Start Distance=100, End Distance=3000, Minimum Output=0, Maximum Output=1, Clamp Output=true)
      )
      False Float=1
     ))
    ScaleA=true
    ScaleRGB=true
   ) {module=ScaleColor001}
  ]
  rendererClasses [&NiagaraSpriteRendererProperties]
  renderers [{rendererClass=&NiagaraSpriteRendererProperties}]
 }
 Ripple {
  simTarget=GPUComputeSim
  emitterUpdateScript [
   &EmitterState()
   &SpawnBurst_Instantaneous(Spawn Count=1, Use Spawn Probability=false) {enabled=false}
   &SpawnRate(SpawnRate=20, Use Spawn Probability=false)
  ]
  particleSpawnScript [
   &InitializeParticle(
    Color=<rgba 0.2604 0.652 1 1>
    Lifetime=0.6
    Position Offset=&RandomRangeVector(Minimum=<-150 -150 0.1>, Maximum=<150 150 0.1>)
    Sprite Rotation Mode=::NewEnumerator2
    Sprite Size Mode=::NewEnumerator3
    Uniform Sprite Size=100
    UsePositionOffset=true
    Write Material Random=true
   ) {module=InitializeParticle001}
   /SetVariables(Particles.SpriteFacing=<0 0 1>) {isSetParameters=true, module=SetVariables_#2}
   /SetVariables(Particles.StaticPos:NiagaraPosition=$Particles.Position) {isSetParameters=true, module=SetVariables_#3}
   &AddVelocity(
    Cone Angle=&RandomRangeFloat(Minimum=10, Maximum=35)
    Cone Axis=<0 0 1>
    Module.UseConeFalloff=false
    Module.UseVelDistribution=false
    Override Seed=false
    Velocity Mode=::NewEnumerator2
    Velocity Speed=&Multiply_Float(A=&RandomRangeFloat(Minimum=1, Maximum=3), B=60)
   )
   &SubUVAnimation(
    Evaluation Type=::NewEnumerator0
    "Sprite Renderer "=DataInterface() {propertyValues {SpriteRenderer=&self:Ripple_2.NiagaraSpriteRendererProperties_1}}
    SubUV Animation Mode:ENiagara_SubUVLookupModeV2=::NewEnumerator2
    UseEndFrame=false
    UseStartFrame=false
   )
  ]
  particleUpdateScript [
   &ParticleState()
   &ScaleSpriteSize(
    Uniform Curve Scale=&Multiply_Float(
     A=&Lerp_Float(
      A=&RandomRangeFloat(Minimum=0.25, Maximum=0.3)
      B=&RandomRangeFloat(Minimum=0.5, Maximum=0.6)
      Alpha=&FloatFromCurve(
       FloatCurve=DataInterface() {propertyValues {Curve {keys [<at 0 0.5>,<at 0.8 0.5>,<at 0.85 1>]}}}
       CurveIndex=&RandomRangeFloat(Minimum=0, Maximum=1)
      )
     )
     B=&RandomRangeFloat(Minimum=0.5, Maximum=0.65)
    )
    Uniform Curve Sprite Scale=DataInterface() {propertyValues {Curve {keys [
        {arriveTangent=1.4867, interpMode=RCIM_Cubic, leaveTangent=1.4867, tangentMode=RCTM_User, time=0, value=0}
        {arriveTangent=0.501, interpMode=RCIM_Cubic, leaveTangent=0.501, tangentMode=RCTM_User, time=1, value=1}
       ]}}}
   )
   &ScaleColor(
    Scale Alpha=&FloatFromCurve(
     FloatCurve=DataInterface() {propertyValues {Curve {keys [
         {arriveTangent=-0.601, interpMode=RCIM_Cubic, leaveTangent=-0.601, tangentMode=RCTM_User, time=0, value=1}
         {arriveTangent=-1.4384, interpMode=RCIM_Cubic, leaveTangent=-1.4384, tangentMode=RCTM_User, time=1, value=0}
        ]}}}
     Scale Curve=&Multiply_Float(A=0.65, B=&FN_FrontEndFloatModifier(Gameplay=1, Front End=0.85))
    )
    Scale RGB=&MakeVectorFromLinearColorRGB(Color=<rgba 1 1 1 1>)
    ScaleA=true
    ScaleRGB=false
   )
   &FN_Sprite_Attributes(Create Position Attribute=true, Position:Vector3f=$Particles.StaticPos, Write to Module Namespace=true)
   &GravityForce()
   &SolveForcesAndVelocity(Clamp Velocity=false, Limit Acceleration=false)
   &FN_Sprite_Attributes(
    Color=&MultiplyLinearColor(
     Linear Color=&FN_Select_Color_from_Time_Of_Day(Morning=<rgba 0.1823 0.6152 1 1>, Day=<rgba 0.1823 0.6152 1 1>, Evening=<rgba 0.3438 0.6912 1 1>, Night=<rgba 0 0.55 1.5 1>)
     Scale Factor=&MakeVector4(
      X=1
      Y=1
      Z=1
      W=&MakeFloatFromLinearColor(LinearColor=$Particles.Color, Channel:ENiagaraLinearColor_Channels=::NewEnumerator3)
     )
    )
    Create Color Attribute=true
    Create SpriteSize Attribute=true
    SpriteSize=&Multiply_Vector2D_ByFloat(Float=&FN_FrontEndFloatModifier(Gameplay=1, Front End=0.75), Vector2D=$Particles.SpriteSize)
    Write to Module Namespace=true
   ) {module=FN_Sprite_Attributes001}
  ]
  rendererClasses [&NiagaraSpriteRendererProperties]
  renderers [
   {rendererClass=&NiagaraSpriteRendererProperties}
   {rendererClass=&NiagaraSpriteRendererProperties, rendererIndex=1}
  ]
 }
 Ripple001 {
  emitterUpdateScript [
   &EmitterState()
   &SpawnBurst_Instantaneous(Spawn Count=1, Use Spawn Probability=false) {enabled=false}
   &SpawnRate(SpawnRate=5, Use Spawn Probability=false)
  ]
  particleSpawnScript [
   &InitializeParticle(
    Color=<rgba 0.2604 0.652 1 1>
    Lifetime=0.6
    Position Offset=&RandomRangeVector(Minimum=<-150 -150 0.1>, Maximum=<150 150 0.1>)
    Sprite Rotation Mode=::NewEnumerator2
    Sprite Size Mode=::NewEnumerator3
    Uniform Sprite Size=100
    UsePositionOffset=true
    Write Material Random=true
   ) {module=InitializeParticle001}
   /SetVariables(Particles.SpriteFacing=<0 0 1>) {isSetParameters=true, module=SetVariables_#2}
   /SetVariables(Particles.StaticPos:NiagaraPosition=$Particles.Position) {isSetParameters=true, module=SetVariables_#3}
   &AddVelocity(
    Cone Angle=&RandomRangeFloat(Minimum=10, Maximum=35)
    Cone Axis=<0 0 1>
    Module.UseConeFalloff=false
    Module.UseVelDistribution=false
    Override Seed=false
    Velocity Mode=::NewEnumerator2
    Velocity Speed=&Multiply_Float(A=&RandomRangeFloat(Minimum=1, Maximum=3), B=80)
   )
   &SubUVAnimation(
    Evaluation Type=::NewEnumerator0
    "Sprite Renderer "=DataInterface() {propertyValues {SpriteRenderer=&self:Ripple001_7.NiagaraSpriteRendererProperties_1}}
    SubUV Animation Mode:ENiagara_SubUVLookupModeV2=::NewEnumerator2
    UseEndFrame=false
    UseStartFrame=false
   )
  ]
  particleUpdateScript [
   &ParticleState()
   &ScaleSpriteSize(
    Uniform Curve Scale=&Multiply_Float(
     A=&Lerp_Float(
      A=&RandomRangeFloat(Minimum=0.25, Maximum=0.3)
      B=&RandomRangeFloat(Minimum=0.5, Maximum=0.6)
      Alpha=&FloatFromCurve(
       FloatCurve=DataInterface() {propertyValues {Curve {keys [<at 0 0.5>,<at 0.8 0.5>,<at 0.85 1>]}}}
       CurveIndex=&RandomRangeFloat(Minimum=0, Maximum=1)
      )
     )
     B=&RandomRangeFloat(Minimum=0.5, Maximum=0.65)
    )
    Uniform Curve Sprite Scale=DataInterface() {propertyValues {Curve {keys [
        {arriveTangent=1.4867, interpMode=RCIM_Cubic, leaveTangent=1.4867, tangentMode=RCTM_User, time=0, value=0}
        {arriveTangent=0.501, interpMode=RCIM_Cubic, leaveTangent=0.501, tangentMode=RCTM_User, time=1, value=1}
       ]}}}
   )
   &ScaleColor(
    Scale Alpha=&FloatFromCurve(
     FloatCurve=DataInterface() {propertyValues {Curve {keys [
         {arriveTangent=-0.601, interpMode=RCIM_Cubic, leaveTangent=-0.601, tangentMode=RCTM_User, time=0, value=1}
         {arriveTangent=-1.4384, interpMode=RCIM_Cubic, leaveTangent=-1.4384, tangentMode=RCTM_User, time=1, value=0}
        ]}}}
     Scale Curve=&Multiply_Float(A=0.65, B=&FN_FrontEndFloatModifier(Gameplay=1, Front End=0.85))
    )
    Scale RGB=&MakeVectorFromLinearColorRGB(Color=<rgba 1 1 1 1>)
    ScaleA=true
    ScaleRGB=false
   )
   &FN_Sprite_Attributes(Create Position Attribute=true, Position:Vector3f=$Particles.StaticPos, Write to Module Namespace=true)
   &GravityForce()
   &SolveForcesAndVelocity(Clamp Velocity=false, Limit Acceleration=false)
   &FN_Sprite_Attributes(
    Color=&MultiplyLinearColor(
     Linear Color=&FN_Select_Color_from_Time_Of_Day(Morning=<rgba 0.1823 0.6152 1 1>, Day=<rgba 0.1823 0.6152 1 1>, Evening=<rgba 0.3438 0.6912 1 1>, Night=<rgba 0 0.55 1.5 1>)
     Scale Factor=&MakeVector4(
      X=1
      Y=1
      Z=1
      W=&MakeFloatFromLinearColor(LinearColor=$Particles.Color, Channel:ENiagaraLinearColor_Channels=::NewEnumerator0)
     )
    )
    Create Color Attribute=true
    Create SpriteSize Attribute=true
    SpriteSize=&Multiply_Vector2D_ByFloat(
     Float=&Multiply_Float(A=1.5, B=&FN_FrontEndFloatModifier(Gameplay=1, Front End=0.75))
     Vector2D=$Particles.SpriteSize
    )
    Write to Module Namespace=true
   ) {module=FN_Sprite_Attributes001}
  ]
  rendererClasses [&NiagaraSpriteRendererProperties]
  renderers [
   {rendererClass=&NiagaraSpriteRendererProperties}
   {rendererClass=&NiagaraSpriteRendererProperties, rendererIndex=1}
  ]
 }
}
summary {
 systemName=NS_Emote_Precipitation_Rain
 emitters [
  {emitterName=Raindrops_GPU, simTarget=GPUComputeSim, rendererClasses [&NiagaraSpriteRendererProperties]}
  {emitterName=Ripple, simTarget=GPUComputeSim, rendererClasses [&NiagaraSpriteRendererProperties]}
  {emitterName=Raindrops_Low, rendererClasses [&NiagaraSpriteRendererProperties]}
  {emitterName=Ripple001, rendererClasses [&NiagaraSpriteRendererProperties]}
 ]
 userVariables [
  User.bIsFrontEnd {type=NiagaraBool}
  User.bIsFrontEndPreview {type=NiagaraBool}
  User.BoxSize {type=Vector3f, description="", defaultValue {struct=Vector3f, value=<200 200 0>}}
  User.Color {type=LinearColor, description="", defaultValue {struct=LinearColor, value=<rgba 0.2552 0.7072 1 1>}}
  User.CustomColorMult {type=NiagaraFloat, defaultValue=1, description=""}
  User.SpawnRate {type=NiagaraFloat, defaultValue=1, description=""}
  User.UseCustomColor {type=NiagaraFloat, defaultValue=1, description=""}
 ]
}
