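# Niagara system for the large blue water splash in the Precipitation emote.
# Burst sprite emitters layer SubUV splashes (Mid, Wide, Thick) and small droplets, launched from a small shape with gravity, plus one upward-facing ripple that expands and fades.
# User.Intensity scales launch speeds, and time of day adjusts splash and droplet color.
# On selected quality levels alpha is boosted, increasing with distance for splash layers; all but the ripple distance-cull.
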NiagaraSystem /BRCosmetics/Animation/Game/MainPlayer/Emotes/Precipitation/CMM/FX/NS_Emote_Precipitation_Splash_Large
dependencies {
 usedDataInterfaces [
  &NiagaraDataInterfaceSpriteRendererInfo
  &NiagaraDataInterfaceCurve
  &NiagaraDataInterfacePlatformSet
 ]
 usedDynamicInputs [&UniformRangedFloat]
 usedModules [
  &SystemState
  &self:SystemSpawnScript.SystemScriptSource.SystemScriptGraph.NiagaraNodeAssignment_1.SetVariables_#1
 ]
}
emitters {
 Dropplets {
  emitterUpdateScript [
   &EmitterState(Enable Distance Culling=true, MaxDistance=9000, Module.bUseMaxDistance=true, Module.bUseMinDistance=false, Scalability Mode=::NewEnumerator1)
   &SpawnBurst_Instantaneous(Spawn Count=35, Use Spawn Probability=false, UseLoopCountLimit=false)
  ]
  particleSpawnScript [
   &InitializeParticle(
    Color=<rgba 0.5312 0.7794 1 1>
    Lifetime Max=0.5
    Lifetime Min=0.25
    Lifetime Mode=::NewEnumerator1
    Sprite Rotation Mode=::NewEnumerator2
    Sprite Size Max=<4 8>
    Sprite Size Min=<2 4>
    Sprite Size Mode=::NewEnumerator2
    Sprite UV Mode=::NewEnumerator3
    UsePositionOffset=false
    Write Material Random=true
   )
   &ShapeLocation(
    MaskSpawnGroup=false
    Override Seed=false
    Ring Radius=7
    Rotation Angle Type:ENiagara_AngleInput=::NewEnumerator0
    Rotation Mode=::NewEnumerator1
    Shape Primitive=::NewEnumerator4
    UseOwnerScale=true
   )
   &AddVelocity(
    Module.UseConeFalloff=false
    Module.UseVelDistribution=false
    Override Seed=false
    Velocity Mode=::NewEnumerator2
    Velocity Speed=&Multiply_Float(A=&RandomRangeFloat(Minimum=110, Maximum=350), B=$User.Intensity)
   ) {module=AddVelocity001}
   &AddVelocity(
    Module.Use Falloff Distance=false
    Origin Offset=<-2 0 0>
    Override Seed=false
    Velocity Mode=::NewEnumerator1
    Velocity Speed=&Multiply_Float(A=&RandomRangeFloat(Minimum=75, Maximum=250), B=$User.Intensity)
   )
   &CameraOffset(Camera Offset Amount=15)
  ]
  particleUpdateScript [
   &ParticleState()
   &GravityForce(Gravity=<0 0 -1000>)
   &Drag(Use Linear Drag=true, Use Rotational Drag=false)
   &SolveForcesAndVelocity(Clamp Velocity=false, Limit Acceleration=false)
   &ScaleSpriteSize(Uniform Curve Sprite Scale=DataInterface() {propertyValues {Curve {keys [
        {arriveTangent=4.1785, interpMode=RCIM_Cubic, leaveTangent=4.1785, tangentMode=RCTM_User, time=0, value=0.15}
        {arriveTangent=1.1179, interpMode=RCIM_Cubic, leaveTangent=1.1179, tangentMode=RCTM_User, time=0.2042, value=0.6228}
        {arriveTangent=0.0224, interpMode=RCIM_Cubic, leaveTangent=0.0224, tangentMode=RCTM_User, time=1, value=1}
       ]}}})
   &ScaleColor(
    Scale Alpha=&FloatFromCurve(
     FloatCurve=DataInterface() {propertyValues {Curve {keys [<at 0 1 RCIM_Cubic>, {arriveTangent=-2.0383, interpMode=RCIM_Cubic, leaveTangent=-2.0383, tangentMode=RCTM_User, time=1, value=0}]}}}
     Scale Curve=&FNTimeOfDayScalar(Night Scale=0.2, Morning Scale=0.2, Day Scale=0.2, Evening Scale=0.2)
    )
    Scale RGB=&Multiply_VectorByFloat(
     Vector=&FN_Select_Vector_from_Time_Of_Day(Morning=<1 1 1>, Day=<1 1 1>, Evening=<1 1 1>, Night=<0.8 1 1.35>)
     Float=&FNTimeOfDayScalar(Night Scale=0.5, Morning Scale=0.75, Evening Scale=0.75)
    )
    ScaleA=true
    ScaleRGB=true
   )
   &ScaleColor(
    Scale Alpha=&MakeCustomFloatFromBool(
     Bool=&IsPlatformSetActive(New Platform Set=DataInterface() {propertyValues {Platforms {qualityLevelMask=-30}}})
     True Float=10
     False Float=1
    )
    ScaleA=true
    ScaleRGB=true
   ) {module=ScaleColor001}
  ]
  rendererClasses [&NiagaraSpriteRendererProperties]
  renderers [{rendererClass=&NiagaraSpriteRendererProperties}]
 }
 Ripple {
  emitterUpdateScript [
   &EmitterState()
   &SpawnBurst_Instantaneous(Spawn Count=1, Use Spawn Probability=false)
   &SpawnRate(
    SpawnRate=&FloatFromCurve(FloatCurve=DataInterface() {propertyValues {Curve {keys [<at 0 1>,<at 0.45 0>]}}}, CurveIndex=$Emitter.NormalizedLoopAge, Scale Curve=16)
    Use Spawn Probability=false
   ) {enabled=false}
  ]
  particleSpawnScript [
   &InitializeParticle(
    Color=<rgba 0.2604 0.652 1 1>
    Lifetime=0.6
    Sprite Rotation Mode=::NewEnumerator2
    Sprite Size Mode=::NewEnumerator3
    Uniform Sprite Size=&MakeFloatFromVector2d(Vector 2D=&Vector2DFromFloat(Value=&FloatFromCurve(FloatCurve=DataInterface() {propertyValues {Curve {keys [<at 0 1>,<at 0.5 0>]}}}, CurveIndex=$Emitter.NormalizedLoopAge, Scale Curve=100)))
    UsePositionOffset=false
    Write Material Random=true
   ) {module=InitializeParticle001}
   /SetVariables(Particles.SpriteFacing=<0 0 1>) {isSetParameters=true}
  ]
  particleUpdateScript [
   &ParticleState()
   &ScaleSpriteSize(Uniform Curve Sprite Scale=DataInterface() {propertyValues {Curve {keys [
        {arriveTangent=1.4867, interpMode=RCIM_Cubic, leaveTangent=1.4867, tangentMode=RCTM_User, time=0, value=0}
        {arriveTangent=0.501, interpMode=RCIM_Cubic, leaveTangent=0.501, tangentMode=RCTM_User, time=1, value=1}
       ]}}})
   &ScaleColor(
    Scale Alpha=&FloatFromCurve(
     FloatCurve=DataInterface() {propertyValues {Curve {keys [
         {arriveTangent=-0.601, interpMode=RCIM_Cubic, leaveTangent=-0.601, tangentMode=RCTM_User, time=0, value=1}
         {arriveTangent=-1.4384, interpMode=RCIM_Cubic, leaveTangent=-1.4384, tangentMode=RCTM_User, time=1, value=0}
        ]}}}
     Scale Curve=&FloatFromCurve(
      FloatCurve=DataInterface() {propertyValues {Curve {keys [
          {arriveTangent=-0.4982, interpMode=RCIM_Cubic, leaveTangent=-0.4982, tangentMode=RCTM_User, time=0, value=1}
          {arriveTangent=-1.3938, interpMode=RCIM_Cubic, leaveTangent=-1.3938, tangentMode=RCTM_User, time=0.6, value=0}
         ]}}}
      CurveIndex=$Emitter.NormalizedLoopAge
      Scale Curve=0.8
     )
    )
    Scale RGB=&MakeVectorFromLinearColorRGB(Color=<rgba 1 1 1 1>)
    ScaleA=true
    ScaleRGB=false
   )
   &ScaleColor(
    Scale Alpha=&MakeCustomFloatFromBool(
     Bool=&IsPlatformSetActive(New Platform Set=DataInterface() {propertyValues {Platforms {qualityLevelMask=-30}}})
     True Float=100
     False Float=1
    )
    ScaleA=true
    ScaleRGB=true
   ) {module=ScaleColor001}
  ]
  rendererClasses [&NiagaraSpriteRendererProperties]
  renderers [{rendererClass=&NiagaraSpriteRendererProperties}]
 }
 SplashMid {
  emitterUpdateScript [
   &EmitterState(Enable Distance Culling=true, MaxDistance=9000, Module.bUseMaxDistance=true, Module.bUseMinDistance=false, Scalability Mode=::NewEnumerator1)
   &SpawnBurst_Instantaneous(Spawn Count=2, Use Spawn Probability=false, UseLoopCountLimit=false)
  ]
  particleSpawnScript [
   &InitializeParticle(
    Color=&FN_Select_Color_from_Time_Of_Day(Morning=<rgba 0.1823 0.6152 1 1>, Day=<rgba 0.1823 0.6152 1 1>, Evening=<rgba 0.3438 0.6912 1 1>, Night=<rgba 0 0.55 1.5 1>)
    Lifetime Max=0.5
    Lifetime Min=0.3
    Lifetime Mode=::NewEnumerator1
    Sprite Rotation Mode=::NewEnumerator2
    Sprite Size Mode=::NewEnumerator1
    Sprite UV Mode=::NewEnumerator3
    Uniform Sprite Size Max=60
    Uniform Sprite Size Min=45
    UsePositionOffset=false
    Write Material Random=true
   )
   &SubUVAnimation(
    "Sprite Renderer "=DataInterface() {propertyValues {SpriteRenderer=&self:SplashMid_0.NiagaraSpriteRendererProperties_0}}
    SubUV Animation Mode:ENiagara_SubUVLookupModeV2=::NewEnumerator2
    UseEndFrame=false
    UseStartFrame=false
   )
   &ShapeLocation(
    MaskSpawnGroup=false
    Override Seed=false
    Ring Radius=7
    Rotation Angle Type:ENiagara_AngleInput=::NewEnumerator0
    Rotation Axis=<0 0 0>
    Rotation Coordinate Space=::Simulation
    Rotation Mode=::NewEnumerator1
    Shape Primitive=::NewEnumerator4
    UseOwnerScale=true
   )
   &AddVelocity(
    Cone Angle=30
    Module.UseConeFalloff=false
    Module.UseVelDistribution=false
    Override Seed=false
    Velocity Mode=::NewEnumerator2
    Velocity Speed=&Multiply_Float(A=&RandomRangeFloat(Minimum=25, Maximum=250), B=$User.Intensity)
   ) {module=AddVelocity001}
   &AddVelocity(
    Module.Use Falloff Distance=false
    Origin Offset=<-10 0 0>
    Override Seed=false
    Velocity Mode=::NewEnumerator1
    Velocity Speed=&Multiply_Float(A=&RandomRangeFloat(Minimum=50, Maximum=250), B=$User.Intensity)
   )
   &CameraOffset(Camera Offset Amount=&RandomRangeFloat(Minimum=10, Maximum=15))
  ]
  particleUpdateScript [
   &ParticleState()
   &GravityForce(Gravity=<0 0 -1200>)
   &Drag(Use Linear Drag=true, Use Rotational Drag=false)
   &SolveForcesAndVelocity(Clamp Velocity=false, Limit Acceleration=false)
   &ScaleSpriteSize(
    Uniform Curve Scale=0.7
    Uniform Curve Sprite Scale=DataInterface() {propertyValues {Curve {keys [
        {arriveTangent=4.1785, interpMode=RCIM_Cubic, leaveTangent=4.1785, tangentMode=RCTM_User, time=0, value=0.15}
        {arriveTangent=1.1179, interpMode=RCIM_Cubic, leaveTangent=1.1179, tangentMode=RCTM_User, time=0.2042, value=0.6228}
        {arriveTangent=0.0224, interpMode=RCIM_Cubic, leaveTangent=0.0224, tangentMode=RCTM_User, time=1, value=1}
       ]}}}
   )
   &ScaleColor(
    Scale Alpha=&FloatFromCurve(
     FloatCurve=DataInterface() {propertyValues {Curve {keys [<at 0 1 RCIM_Cubic>, {arriveTangent=-2.0383, interpMode=RCIM_Cubic, leaveTangent=-2.0383, tangentMode=RCTM_User, time=1, value=0}]}}}
     Scale Curve=0.7
    )
    ScaleA=true
    ScaleRGB=true
   )
   &ScaleColor(
    Scale Alpha=&MakeCustomFloatFromBool(
     Bool=&IsPlatformSetActive(New Platform Set=DataInterface() {propertyValues {Platforms {qualityLevelMask=-29}}})
     True Float=&Lerp_Float(
      A=1
      B=10
      Alpha=&FN_LOD_Distance_Remap(Start Distance=100, End Distance=3000, Minimum Output=0, Maximum Output=1, Clamp Output=true)
     )
     False Float=1
    )
    ScaleA=true
    ScaleRGB=true
   ) {module=ScaleColor001}
  ]
  rendererClasses [&NiagaraSpriteRendererProperties]
  renderers [{rendererClass=&NiagaraSpriteRendererProperties}]
 }
 SplashThick {
  emitterUpdateScript [
   &EmitterState(Enable Distance Culling=true, MaxDistance=9000, Module.bUseMaxDistance=true, Module.bUseMinDistance=false, Scalability Mode=::NewEnumerator1)
   &SpawnBurst_Instantaneous(Use Spawn Probability=false, UseLoopCountLimit=false)
  ]
  particleSpawnScript [
   &InitializeParticle(
    Color=&FN_Select_Color_from_Time_Of_Day(Morning=<rgba 0.1823 0.6152 1 1>, Day=<rgba 0.1823 0.6152 1 1>, Evening=<rgba 0.3438 0.6912 1 1>, Night=<rgba 0 0.55 1.5 1>)
    Lifetime Max=0.75
    Lifetime Min=0.5
    Lifetime Mode=::NewEnumerator1
    Sprite Rotation Mode=::NewEnumerator2
    Sprite Size Mode=::NewEnumerator3
    Sprite UV Mode=::NewEnumerator3
    Uniform Sprite Size=90
    UsePositionOffset=false
    Write Material Random=true
   )
   &SubUVAnimation(
    "Sprite Renderer "=DataInterface() {propertyValues {SpriteRenderer=&self:SplashThick_2.NiagaraSpriteRendererProperties_0}}
    SubUV Animation Mode:ENiagara_SubUVLookupModeV2=::NewEnumerator2
    UseEndFrame=false
    UseStartFrame=false
   )
   &ShapeLocation(
    MaskSpawnGroup=false
    Override Seed=false
    Ring Radius=7
    Rotation Angle Type:ENiagara_AngleInput=::NewEnumerator0
    Rotation Axis=<0 0 0>
    Rotation Coordinate Space=::Simulation
    Rotation Mode=::NewEnumerator1
    Shape Primitive=::NewEnumerator4
    UseOwnerScale=true
   )
   &AddVelocity(
    Cone Angle=0
    Module.UseConeFalloff=false
    Module.UseVelDistribution=false
    Override Seed=false
    Velocity Mode=::NewEnumerator2
    Velocity Speed=&Multiply_Float(A=&RandomRangeFloat(Minimum=20, Maximum=160), B=$User.Intensity)
   ) {module=AddVelocity001}
   &CameraOffset(Camera Offset Amount=7)
  ]
  particleUpdateScript [
   &ParticleState()
   &GravityForce(Gravity=<0 0 -950>)
   &SolveForcesAndVelocity(Clamp Velocity=false, Limit Acceleration=false)
   &ScaleSpriteSize(Uniform Curve Sprite Scale=DataInterface() {propertyValues {Curve {keys [
        {arriveTangent=4.1785, interpMode=RCIM_Cubic, leaveTangent=4.1785, tangentMode=RCTM_User, time=0, value=0.15}
        {arriveTangent=1.1179, interpMode=RCIM_Cubic, leaveTangent=1.1179, tangentMode=RCTM_User, time=0.2042, value=0.6228}
        {arriveTangent=0.0224, interpMode=RCIM_Cubic, leaveTangent=0.0224, tangentMode=RCTM_User, time=1, value=1}
       ]}}})
   &ScaleColor(
    Scale Alpha=&FloatFromCurve(FloatCurve=DataInterface() {propertyValues {Curve {keys [<at 0 1 RCIM_Cubic>, {arriveTangent=-2.0383, interpMode=RCIM_Cubic, leaveTangent=-2.0383, tangentMode=RCTM_User, time=1, value=0}]}}})
    Scale RGB=&Multiply_VectorByFloat(
     Vector=&FN_Select_Vector_from_Time_Of_Day(Morning=<1 1 1>, Day=<1 1 1>, Evening=<1 1 1>, Night=<1 1 1>)
     Float=&FNTimeOfDayScalar(Night Scale=1, Morning Scale=1, Evening Scale=1)
    )
    ScaleA=true
    ScaleRGB=true
   )
   &ScaleColor(
    Scale Alpha=&MakeCustomFloatFromBool(
     Bool=&IsPlatformSetActive(New Platform Set=DataInterface() {propertyValues {Platforms {qualityLevelMask=-29}}})
     True Float=&Lerp_Float(
      A=1
      B=10
      Alpha=&FN_LOD_Distance_Remap(Start Distance=100, End Distance=3000, Minimum Output=0, Maximum Output=1, Clamp Output=true)
     )
     False Float=1
    )
    ScaleA=true
    ScaleRGB=true
   ) {module=ScaleColor001}
  ]
  rendererClasses [&NiagaraSpriteRendererProperties]
  renderers [{rendererClass=&NiagaraSpriteRendererProperties}]
 }
 SplashWide {
  emitterUpdateScript [
   &EmitterState(Enable Distance Culling=true, MaxDistance=9000, Module.bUseMaxDistance=true, Module.bUseMinDistance=false, Scalability Mode=::NewEnumerator1)
   &SpawnBurst_Instantaneous(Spawn Count=4, Use Spawn Probability=false, UseLoopCountLimit=false)
  ]
  particleSpawnScript [
   &InitializeParticle(
    Color=&FN_Select_Color_from_Time_Of_Day(Morning=<rgba 0.1823 0.6152 1 1>, Day=<rgba 0.1823 0.6152 1 1>, Evening=<rgba 0.3438 0.6912 1 1>, Night=<rgba 0 0.55 1.5 1>)
    Lifetime Max=0.75
    Lifetime Min=0.5
    Lifetime Mode=::NewEnumerator1
    Sprite Rotation Mode=::NewEnumerator2
    Sprite Size Mode=::NewEnumerator1
    Sprite UV Mode=::NewEnumerator3
    Uniform Sprite Size Max=40
    Uniform Sprite Size Min=32.5
    UsePositionOffset=false
    Write Material Random=true
   )
   &SubUVAnimation(
    "Sprite Renderer "=DataInterface() {propertyValues {SpriteRenderer=&self:SplashWide_1.NiagaraSpriteRendererProperties_0}}
    SubUV Animation Mode:ENiagara_SubUVLookupModeV2=::NewEnumerator2
    UseEndFrame=false
    UseStartFrame=false
   )
   &ShapeLocation(
    MaskSpawnGroup=false
    Override Seed=false
    Ring Radius=7
    Rotation Angle Type:ENiagara_AngleInput=::NewEnumerator0
    Rotation Axis=<0 0 0>
    Rotation Coordinate Space=::Simulation
    Rotation Mode=::NewEnumerator1
    Shape Primitive=::NewEnumerator4
    UseOwnerScale=true
   )
   &AddVelocity(
    Cone Angle=30
    Module.UseConeFalloff=false
    Module.UseVelDistribution=false
    Override Seed=false
    Velocity Mode=::NewEnumerator2
    Velocity Speed=&Multiply_Float(A=&RandomRangeFloat(Minimum=50, Maximum=300), B=$User.Intensity)
   ) {module=AddVelocity001}
   &AddVelocity(
    Module.Use Falloff Distance=false
    Origin Offset=<-10 0 0>
    Override Seed=false
    Velocity Mode=::NewEnumerator1
    Velocity Speed=&Multiply_Float(A=&RandomRangeFloat(Minimum=100, Maximum=280), B=$User.Intensity)
   )
   &CameraOffset(Camera Offset Amount=&RandomRangeFloat(Minimum=10, Maximum=15))
  ]
  particleUpdateScript [
   &ParticleState()
   &GravityForce(Gravity=<0 0 -1000>)
   &Drag(Use Linear Drag=true, Use Rotational Drag=false)
   &SolveForcesAndVelocity(Clamp Velocity=false, Limit Acceleration=false)
   &ScaleSpriteSize(Uniform Curve Sprite Scale=DataInterface() {propertyValues {Curve {keys [
        {arriveTangent=4.1785, interpMode=RCIM_Cubic, leaveTangent=4.1785, tangentMode=RCTM_User, time=0, value=0.15}
        {arriveTangent=1.1179, interpMode=RCIM_Cubic, leaveTangent=1.1179, tangentMode=RCTM_User, time=0.2042, value=0.6228}
        {arriveTangent=0.0224, interpMode=RCIM_Cubic, leaveTangent=0.0224, tangentMode=RCTM_User, time=1, value=1}
       ]}}})
   &ScaleColor(
    Scale Alpha=&FloatFromCurve(FloatCurve=DataInterface() {propertyValues {Curve {keys [<at 0 1 RCIM_Cubic>, {arriveTangent=-2.0383, interpMode=RCIM_Cubic, leaveTangent=-2.0383, tangentMode=RCTM_User, time=1, value=0}]}}})
    Scale RGB=&Multiply_VectorByFloat(
     Vector=&FN_Select_Vector_from_Time_Of_Day(Morning=<1 1 1>, Day=<1 1 1>, Evening=<1 1 1>, Night=<1 1 1>)
     Float=&FNTimeOfDayScalar(Night Scale=1, Morning Scale=1, Evening Scale=1)
    )
    ScaleA=true
    ScaleRGB=true
   )
   &ScaleColor(
    Scale Alpha=&MakeCustomFloatFromBool(
     Bool=&IsPlatformSetActive(New Platform Set=DataInterface() {propertyValues {Platforms {qualityLevelMask=-29}}})
     True Float=&Lerp_Float(
      A=1
      B=5
      Alpha=&FN_LOD_Distance_Remap(Start Distance=100, End Distance=3000, Minimum Output=0, Maximum Output=1, Clamp Output=true)
     )
     False Float=1
    )
    ScaleA=true
    ScaleRGB=true
   ) {module=ScaleColor001}
  ]
  rendererClasses [&NiagaraSpriteRendererProperties]
  renderers [{rendererClass=&NiagaraSpriteRendererProperties}]
 }
}
summary {
 systemName=NS_Emote_Precipitation_Splash_Large
 emitters [
  {emitterName=SplashMid, rendererClasses [&NiagaraSpriteRendererProperties]}
  {emitterName=SplashWide, rendererClasses [&NiagaraSpriteRendererProperties]}
  {emitterName=Dropplets, rendererClasses [&NiagaraSpriteRendererProperties]}
  {emitterName=SplashThick, rendererClasses [&NiagaraSpriteRendererProperties]}
  {emitterName=Ripple, rendererClasses [&NiagaraSpriteRendererProperties]}
 ]
 userVariables [
  User.Intensity {type=NiagaraFloat, defaultValue=0.68, description=""}
  User.Variant {type=NiagaraFloat, defaultValue=0, description=""}
 ]
}
