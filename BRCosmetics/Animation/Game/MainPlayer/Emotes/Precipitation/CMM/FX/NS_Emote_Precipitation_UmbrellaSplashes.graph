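# Niagara system for the Precipitation emote that draws short-lived rain splash ripple sprites on the Melancholy_Gadget umbrella prop.
# Two near-identical emitters, Ripple on GPU and Ripple001 on default CPU at a lower gameplay rate, spawn from filtered umbrella bones on the skinned mesh.
# Sprites face up, grow with random scale and fade over life.
# Spawn rate and size adapt to front end versus gameplay, while color follows time of day.
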NiagaraSystem /BRCosmetics/Animation/Game/MainPlayer/Emotes/Precipitation/CMM/FX/NS_Emote_Precipitation_UmbrellaSplashes
dependencies {
 usedDataInterfaces [
  &NiagaraDataInterfaceSpriteRendererInfo
  &NiagaraDataInterfaceSkeletalMesh
  &NiagaraDataInterfaceCurve
 ]
 usedDynamicInputs [&UniformRangedFloat]
 usedModules [&SystemState]
}
emitters {
 Ripple {
  simTarget=GPUComputeSim
  emitterUpdateScript [
   &EmitterState()
   &SpawnBurst_Instantaneous(Spawn Count=1, Use Spawn Probability=false) {enabled=false}
   &SpawnRate(SpawnRate=&FN_FrontEndFloatModifier(Gameplay=50, Front End=30), Use Spawn Probability=false)
  ]
  particleSpawnScript [
   &InitializeParticle(
    Color=&FN_Select_Color_from_Time_Of_Day(Morning=<rgba 0.1823 0.6152 1 1>, Day=<rgba 0.1823 0.6152 1 1>, Evening=<rgba 0.3438 0.6912 1 1>, Night=<rgba 0 0.55 1.5 1>)
    Lifetime=0.2
    Position Offset=&RandomRangeVector(Minimum=<-150 -150 0.1>, Maximum=<150 150 0.1>)
    Sprite Rotation Mode=::NewEnumerator2
    Sprite Size Mode=::NewEnumerator3
    Uniform Sprite Size=30
    UsePositionOffset=true
    Write Material Random=true
   ) {module=InitializeParticle001}
   /SetVariables(Particles.SpriteFacing=<0 0 1>) {isSetParameters=true, module=SetVariables_#1}
   /SetVariables(Particles.StaticPos:NiagaraPosition=$Particles.Position) {isSetParameters=true, module=SetVariables_#2}
   &SubUVAnimation(
    Evaluation Type=::NewEnumerator0
    "Sprite Renderer "=DataInterface() {propertyValues {SpriteRenderer=&self:Ripple_2.NiagaraSpriteRendererProperties_1}}
    SubUV Animation Mode:ENiagara_SubUVLookupModeV2=::NewEnumerator2
    UseEndFrame=false
    UseStartFrame=false
   )
   &SkeletalMeshLocation(
    Bone Sampling Mode=::NewEnumerator0
    MaskSpawnGroup=false
    OverrideRandomSeed=false
    Skeletal Mesh=DataInterface() {propertyValues {
      DefaultMesh=/BRCosmetics/Animation/Props/Gadgets_Melancholy/Meshes/Melancholy_Gadget
      ExcludeBoneName=root
      OptionalComponentClass=null
      PreviewMesh=&Melancholy_Gadget
      SkinningMode=SkinOnTheFly
      SoftSourceActor=null
      SourceMode=Default
      UvSetIndex=0
      WholeMeshLOD=-1
      bExcludeBone=false
      bReadDeformedGeometry=true
      bRequireCurrentFrameData=true
      ComponentTags []
      FilteredBones [
       JNT_umbrella_01_middle
       JNT_umbrella_01_end
       JNT_umbrella_02_middle
       JNT_umbrella_02_end
       JNT_umbrella_03_middle
       JNT_umbrella_03_end
       JNT_umbrella_04_middle
       JNT_umbrella_04_end
       JNT_umbrella_05_middle
       JNT_umbrella_05_end
       JNT_umbrella_08_middle
       JNT_umbrella_08_end
       JNT_umbrella_09_middle
       JNT_umbrella_09_end
       JNT_umbrella_10_middle
       JNT_umbrella_10_end
       JNT_umbrella_06_middle
       JNT_umbrella_06_end
       JNT_umbrella_01_start
      ]
      FilteredSockets []
      MeshUserParameter {parameter {name=None, typeDefHandle {registeredTypeIndex=1}, varData []}}
      SamplingRegions []
     }}
   )
   &ShapeLocation(MaskSpawnGroup=false, Override Seed=false, Sphere Radius=0, UseOwnerScale=true)
  ]
  particleUpdateScript [
   &ParticleState()
   &ScaleSpriteSize(
    Uniform Curve Scale=&Multiply_Float(
     A=&Multiply_Float(
      A=&Lerp_Float(
       A=&RandomRangeFloat(Minimum=0.25, Maximum=0.3)
       B=&RandomRangeFloat(Minimum=0.5, Maximum=0.6)
       Alpha=&FloatFromCurve(
        FloatCurve=DataInterface() {propertyValues {Curve {keys [<at 0 0.5>,<at 0.8 0.5>,<at 0.85 1>]}}}
        CurveIndex=&RandomRangeFloat(Minimum=0, Maximum=1)
       )
      )
      B=&RandomRangeFloat(Minimum=0.5, Maximum=0.65)
     )
     B=&FN_FrontEndFloatModifier(Gameplay=1, Front End=1.3)
    )
    Uniform Curve Sprite Scale=DataInterface() {propertyValues {Curve {keys [
        {arriveTangent=1.4867, interpMode=RCIM_Cubic, leaveTangent=1.4867, tangentMode=RCTM_User, time=0, value=0}
        {arriveTangent=0.501, interpMode=RCIM_Cubic, leaveTangent=0.501, tangentMode=RCTM_User, time=1, value=1}
       ]}}}
   )
   &ScaleColor(
    Scale Alpha=&FloatFromCurve(
     FloatCurve=DataInterface() {propertyValues {Curve {keys [
         {arriveTangent=-0.601, interpMode=RCIM_Cubic, leaveTangent=-0.601, tangentMode=RCTM_User, time=0, value=1}
         {arriveTangent=-1.4384, interpMode=RCIM_Cubic, leaveTangent=-1.4384, tangentMode=RCTM_User, time=1, value=0}
        ]}}}
     Scale Curve=0.65
    )
    Scale RGB=&MakeVectorFromLinearColorRGB(Color=<rgba 1 1 1 1>)
    ScaleA=true
    ScaleRGB=false
   )
   &FN_Sprite_Attributes(Create Position Attribute=true, Position:Vector3f=$Particles.StaticPos, Write to Module Namespace=true)
   &FN_Sprite_Attributes(
    Color=&MultiplyLinearColor(
     Linear Color=&FN_Select_Color_from_Time_Of_Day(Morning=<rgba 0.1823 0.6152 1 1>, Day=<rgba 0.1823 0.6152 1 1>, Evening=<rgba 0.3438 0.6912 1 1>, Night=<rgba 0 0.55 1.5 1>)
     Scale Factor=&MakeVector4(
      X=1
      Y=1
      Z=1
      W=&MakeFloatFromLinearColor(LinearColor=$Particles.Color, Channel:ENiagaraLinearColor_Channels=::NewEnumerator0)
     )
    )
    Create Color Attribute=true
    Create SpriteSize Attribute=true
    SpriteSize=&Multiply_Vector2D_ByFloat(Float=1.5, Vector2D=$Particles.SpriteSize)
    Write to Module Namespace=true
   ) {module=FN_Sprite_Attributes001}
   /SetVariables(Particles.FinalPosition:NiagaraPosition=&AddVectorToPosition(
     Position=$Particles.Position
     Vector=&TransformVector(
      Vector=&MakeVector(
       X=0
       Y=0
       Z=&FloatFromCurve(
        FloatCurve=DataInterface() {propertyValues {Curve {keys [{tangentMode=RCTM_User, time=0, value=0}, {tangentMode=RCTM_User, time=1, value=1}]}}}
        Scale Curve=3
       )
      )
      Source Space=::Simulation
      Destination Space=::Local
     )
    )) {isSetParameters=true, module=SetVariables_#3}
  ]
  rendererClasses [&NiagaraSpriteRendererProperties]
  renderers [{rendererClass=&NiagaraSpriteRendererProperties}]
 }
 Ripple001 {
  emitterUpdateScript [
   &EmitterState()
   &SpawnBurst_Instantaneous(Spawn Count=1, Use Spawn Probability=false) {enabled=false}
   &SpawnRate(SpawnRate=&FN_FrontEndFloatModifier(Gameplay=25, Front End=30), Use Spawn Probability=false)
  ]
  particleSpawnScript [
   &InitializeParticle(
    Color=&FN_Select_Color_from_Time_Of_Day(Morning=<rgba 0.1823 0.6152 1 1>, Day=<rgba 0.1823 0.6152 1 1>, Evening=<rgba 0.3438 0.6912 1 1>, Night=<rgba 0 0.55 1.5 1>)
    Lifetime=0.2
    Position Offset=&RandomRangeVector(Minimum=<-150 -150 0.1>, Maximum=<150 150 0.1>)
    Sprite Rotation Mode=::NewEnumerator2
    Sprite Size Mode=::NewEnumerator3
    Uniform Sprite Size=30
    UsePositionOffset=true
    Write Material Random=true
   ) {module=InitializeParticle001}
   /SetVariables(Particles.SpriteFacing=<0 0 1>) {isSetParameters=true, module=SetVariables_#1}
   /SetVariables(Particles.StaticPos:NiagaraPosition=$Particles.Position) {isSetParameters=true, module=SetVariables_#2}
   &SubUVAnimation(
    Evaluation Type=::NewEnumerator0
    "Sprite Renderer "=DataInterface() {propertyValues {SpriteRenderer=&self:Ripple001_1.NiagaraSpriteRendererProperties_1}}
    SubUV Animation Mode:ENiagara_SubUVLookupModeV2=::NewEnumerator2
    UseEndFrame=false
    UseStartFrame=false
   )
   &SkeletalMeshLocation(
    Bone Sampling Mode=::NewEnumerator0
    MaskSpawnGroup=false
    OverrideRandomSeed=false
    Skeletal Mesh=DataInterface() {propertyValues {
      DefaultMesh=&Melancholy_Gadget
      ExcludeBoneName=root
      OptionalComponentClass=null
      PreviewMesh=&Melancholy_Gadget
      SkinningMode=SkinOnTheFly
      SoftSourceActor=null
      SourceMode=Default
      UvSetIndex=0
      WholeMeshLOD=-1
      bExcludeBone=false
      bReadDeformedGeometry=true
      bRequireCurrentFrameData=true
      ComponentTags []
      FilteredBones [
       JNT_umbrella_01_middle
       JNT_umbrella_01_end
       JNT_umbrella_02_middle
       JNT_umbrella_02_end
       JNT_umbrella_03_middle
       JNT_umbrella_03_end
       JNT_umbrella_04_middle
       JNT_umbrella_04_end
       JNT_umbrella_05_middle
       JNT_umbrella_05_end
       JNT_umbrella_08_middle
       JNT_umbrella_08_end
       JNT_umbrella_09_middle
       JNT_umbrella_09_end
       JNT_umbrella_10_middle
       JNT_umbrella_10_end
       JNT_umbrella_06_middle
       JNT_umbrella_06_end
       JNT_umbrella_01_start
      ]
      FilteredSockets []
      MeshUserParameter {parameter {name=None, typeDefHandle {registeredTypeIndex=1}, varData []}}
      SamplingRegions []
     }}
   )
   &ShapeLocation(MaskSpawnGroup=false, Override Seed=false, Sphere Radius=0, UseOwnerScale=true)
  ]
  particleUpdateScript [
   &ParticleState()
   &ScaleSpriteSize(
    Uniform Curve Scale=&Multiply_Float(
     A=&Multiply_Float(
      A=&Lerp_Float(
       A=&RandomRangeFloat(Minimum=0.25, Maximum=0.3)
       B=&RandomRangeFloat(Minimum=0.5, Maximum=0.6)
       Alpha=&FloatFromCurve(
        FloatCurve=DataInterface() {propertyValues {Curve {keys [<at 0 0.5>,<at 0.8 0.5>,<at 0.85 1>]}}}
        CurveIndex=&RandomRangeFloat(Minimum=0, Maximum=1)
       )
      )
      B=&RandomRangeFloat(Minimum=0.5, Maximum=0.65)
     )
     B=&FN_FrontEndFloatModifier(Gameplay=1, Front End=1.3)
    )
    Uniform Curve Sprite Scale=DataInterface() {propertyValues {Curve {keys [
        {arriveTangent=1.4867, interpMode=RCIM_Cubic, leaveTangent=1.4867, tangentMode=RCTM_User, time=0, value=0}
        {arriveTangent=0.501, interpMode=RCIM_Cubic, leaveTangent=0.501, tangentMode=RCTM_User, time=1, value=1}
       ]}}}
   )
   &ScaleColor(
    Scale Alpha=&FloatFromCurve(
     FloatCurve=DataInterface() {propertyValues {Curve {keys [
         {arriveTangent=-0.601, interpMode=RCIM_Cubic, leaveTangent=-0.601, tangentMode=RCTM_User, time=0, value=1}
         {arriveTangent=-1.4384, interpMode=RCIM_Cubic, leaveTangent=-1.4384, tangentMode=RCTM_User, time=1, value=0}
        ]}}}
     Scale Curve=0.65
    )
    Scale RGB=&MakeVectorFromLinearColorRGB(Color=<rgba 1 1 1 1>)
    ScaleA=true
    ScaleRGB=false
   )
   &FN_Sprite_Attributes(Create Position Attribute=true, Position:Vector3f=$Particles.StaticPos, Write to Module Namespace=true)
   &FN_Sprite_Attributes(
    Color=&MultiplyLinearColor(
     Linear Color=&FN_Select_Color_from_Time_Of_Day(Morning=<rgba 0.1823 0.6152 1 1>, Day=<rgba 0.1823 0.6152 1 1>, Evening=<rgba 0.3438 0.6912 1 1>, Night=<rgba 0 0.55 1.5 1>)
     Scale Factor=&MakeVector4(
      X=1
      Y=1
      Z=1
      W=&MakeFloatFromLinearColor(LinearColor=$Particles.Color, Channel:ENiagaraLinearColor_Channels=::NewEnumerator0)
     )
    )
    Create Color Attribute=true
    Create SpriteSize Attribute=true
    SpriteSize=&Multiply_Vector2D_ByFloat(Float=1.5, Vector2D=$Particles.SpriteSize)
    Write to Module Namespace=true
   ) {module=FN_Sprite_Attributes001}
   /SetVariables(Particles.FinalPosition:NiagaraPosition=&AddVectorToPosition(
     Position=$Particles.Position
     Vector=&TransformVector(
      Vector=&MakeVector(
       X=0
       Y=0
       Z=&FloatFromCurve(
        FloatCurve=DataInterface() {propertyValues {Curve {keys [{tangentMode=RCTM_User, time=0, value=0}, {tangentMode=RCTM_User, time=1, value=1}]}}}
        Scale Curve=3
       )
      )
      Source Space=::Simulation
      Destination Space=::Local
     )
    )) {isSetParameters=true, module=SetVariables_#3}
  ]
  rendererClasses [&NiagaraSpriteRendererProperties]
  renderers [{rendererClass=&NiagaraSpriteRendererProperties}]
 }
}
summary {
 systemName=NS_Emote_Precipitation_UmbrellaSplashes
 emitters [
  {emitterName=Ripple, simTarget=GPUComputeSim, rendererClasses [&NiagaraSpriteRendererProperties]}
  {emitterName=Ripple001, rendererClasses [&NiagaraSpriteRendererProperties]}
 ]
 userVariables [
  User.bIsFrontEnd {type=NiagaraBool}
  User.bIsFrontEndPreview {type=NiagaraBool}
  User.BoxSize {type=Vector3f, description="", defaultValue {struct=Vector3f, value=<200 200 0>}}
  User.Color {type=LinearColor, description="", defaultValue {struct=LinearColor, value=<rgba 0.2552 0.7072 1 1>}}
  User.CustomColorMult {type=NiagaraFloat, defaultValue=1, description=""}
  User.SpawnRate {type=NiagaraFloat, defaultValue=1, description=""}
  User.UseCustomColor {type=NiagaraFloat, defaultValue=1, description=""}
 ]
}
